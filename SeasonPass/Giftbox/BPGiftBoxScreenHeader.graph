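# Header widget for the Season Pass gift box screen, showing pass badge, header, subheader, body text and auto-claim status.
# The badge uses the pass viewmodel, or falls back to the gift box image and seasonOwned parameter.
# Pass theme data supplies a stored background material, preferring the giftbox-specific one.
# The autoSpend parameter controls auto-claim visibility, and handler functions play initial-state, staggered intro and fade-out animations.
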WidgetBlueprint /SeasonPass/Giftbox/BPGiftBoxScreenHeader
animations [
 IntroAnim {
  binding_count=5
  length_seconds=0.5
  object=&self:IntroAnim
  sequence {bindings [
    AutoClaimStatus {
     id=#1
     possessed_class=WBP_GiftBox_AutoClaimedStatus_C
     tracks [
      {
       type=MovieSceneFloatTrack
       display_name=Render Opacity
       sections [{type=MovieSceneFloatSection, end_frame=10, start_frame=7, channels [None {type=MovieSceneScriptingFloatChannel, keys [[7,0],[10,1]]}]}]
      }
      {
       type=MovieScene2DTransformTrack
       sections [{
         type=MovieScene2DTransformSection
         end_frame=10
         start_frame=7
         channels [
          Translation.X {type=MovieSceneScriptingFloatChannel, const [0,1]}
          Translation.Y {type=MovieSceneScriptingFloatChannel, keys [[7,-16],[10,0]]}
          Angle {type=MovieSceneScriptingFloatChannel, const [0,1]}
          Scale.X {type=MovieSceneScriptingFloatChannel, const [1,1]}
          Scale.Y {type=MovieSceneScriptingFloatChannel, const [1,1]}
          Shear.X {type=MovieSceneScriptingFloatChannel, const [0,1]}
          Shear.Y {type=MovieSceneScriptingFloatChannel, const [0,1]}
         ]
        }]
      }
     ]
    }
    Text_Body {
     id=#2
     possessed_class=CommonTextBlock
     tracks [
      {
       type=MovieScene2DTransformTrack
       sections [{
         type=MovieScene2DTransformSection
         channels [
          Translation.X {type=MovieSceneScriptingFloatChannel, const [0,2]}
          Translation.Y {type=MovieSceneScriptingFloatChannel, keys [[7,-16],[10,0]]}
          Angle {type=MovieSceneScriptingFloatChannel}
          Scale.X {type=MovieSceneScriptingFloatChannel}
          Scale.Y {type=MovieSceneScriptingFloatChannel}
          Shear.X {type=MovieSceneScriptingFloatChannel}
          Shear.Y {type=MovieSceneScriptingFloatChannel}
         ]
        }]
      }
      {
       type=MovieSceneFloatTrack
       display_name=Render Opacity
       sections [{type=MovieSceneFloatSection, channels [None {type=MovieSceneScriptingFloatChannel, keys [[7,0],[10,1]]}]}]
      }
     ]
    }
    Text_SubHeader {
     id=#3
     possessed_class=CommonTextBlock
     tracks [
      {
       type=MovieScene2DTransformTrack
       sections [{
         type=MovieScene2DTransformSection
         channels [
          Translation.X {type=MovieSceneScriptingFloatChannel, const [0,2]}
          Translation.Y {type=MovieSceneScriptingFloatChannel, keys [[6,-32],[9,0]]}
          Angle {type=MovieSceneScriptingFloatChannel}
          Scale.X {type=MovieSceneScriptingFloatChannel}
          Scale.Y {type=MovieSceneScriptingFloatChannel}
          Shear.X {type=MovieSceneScriptingFloatChannel}
          Shear.Y {type=MovieSceneScriptingFloatChannel}
         ]
        }]
      }
      {
       type=MovieSceneFloatTrack
       display_name=Render Opacity
       sections [{type=MovieSceneFloatSection, channels [None {type=MovieSceneScriptingFloatChannel, keys [[6,0],[9,1]]}]}]
      }
     ]
    }
    SeasonPassBadgeIcon {
     id=#4
     possessed_class=SeasonPassBadgeIcon_C
     tracks [
      {
       type=MovieSceneFloatTrack
       display_name=Render Opacity
       sections [{type=MovieSceneFloatSection, channels [None {type=MovieSceneScriptingFloatChannel, keys [[0,0],[3,1]]}]}]
      }
      {
       type=MovieScene2DTransformTrack
       sections [{
         type=MovieScene2DTransformSection
         channels [
          Translation.X {type=MovieSceneScriptingFloatChannel, keys [[0,-32],[3,0]]}
          Translation.Y {type=MovieSceneScriptingFloatChannel, const [0,2]}
          Angle {type=MovieSceneScriptingFloatChannel}
          Scale.X {type=MovieSceneScriptingFloatChannel}
          Scale.Y {type=MovieSceneScriptingFloatChannel}
          Shear.X {type=MovieSceneScriptingFloatChannel}
          Shear.Y {type=MovieSceneScriptingFloatChannel}
         ]
        }]
      }
     ]
    }
    Text_Header {
     id=#5
     possessed_class=CommonTextBlock
     tracks [
      {
       type=MovieSceneFloatTrack
       display_name=Render Opacity
       sections [{type=MovieSceneFloatSection, channels [None {type=MovieSceneScriptingFloatChannel, keys [[3,0],[6,1]]}]}]
      }
      {
       type=MovieScene2DTransformTrack
       sections [{
         type=MovieScene2DTransformSection
         channels [
          Translation.X {type=MovieSceneScriptingFloatChannel, keys [[3,128],[6,0]]}
          Translation.Y {type=MovieSceneScriptingFloatChannel, const [0,2]}
          Angle {type=MovieSceneScriptingFloatChannel}
          Scale.X {type=MovieSceneScriptingFloatChannel}
          Scale.Y {type=MovieSceneScriptingFloatChannel}
          Shear.X {type=MovieSceneScriptingFloatChannel}
          Shear.Y {type=MovieSceneScriptingFloatChannel}
         ]
        }]
      }
     ]
    }
   ]}
 }
 InitialState {
  binding_count=4
  length_seconds=0.05
  object=&self:InitialState
  sequence {bindings [
    Text_Body {
     id=#2
     possessed_class=CommonTextBlock
     tracks [
      {
       type=MovieScene2DTransformTrack
       sections [{
         type=MovieScene2DTransformSection
         channels [
          Translation.X {type=MovieSceneScriptingFloatChannel, const [0,1]}
          Translation.Y {type=MovieSceneScriptingFloatChannel, const [-32,1]}
          Angle {type=MovieSceneScriptingFloatChannel}
          Scale.X {type=MovieSceneScriptingFloatChannel}
          Scale.Y {type=MovieSceneScriptingFloatChannel}
          Shear.X {type=MovieSceneScriptingFloatChannel}
          Shear.Y {type=MovieSceneScriptingFloatChannel}
         ]
        }]
      }
      {
       type=MovieSceneFloatTrack
       display_name=Render Opacity
       sections [{type=MovieSceneFloatSection, channels [None {type=MovieSceneScriptingFloatChannel, const [0,1]}]}]
      }
     ]
    }
    Text_SubHeader {
     id=#3
     possessed_class=CommonTextBlock
     tracks [
      {
       type=MovieScene2DTransformTrack
       sections [{
         type=MovieScene2DTransformSection
         channels [
          Translation.X {type=MovieSceneScriptingFloatChannel, const [0,1]}
          Translation.Y {type=MovieSceneScriptingFloatChannel, const [-32,1]}
          Angle {type=MovieSceneScriptingFloatChannel}
          Scale.X {type=MovieSceneScriptingFloatChannel}
          Scale.Y {type=MovieSceneScriptingFloatChannel}
          Shear.X {type=MovieSceneScriptingFloatChannel}
          Shear.Y {type=MovieSceneScriptingFloatChannel}
         ]
        }]
      }
      {
       type=MovieSceneFloatTrack
       display_name=Render Opacity
       sections [{type=MovieSceneFloatSection, channels [None {type=MovieSceneScriptingFloatChannel, const [0,1]}]}]
      }
     ]
    }
    Text_Header {
     id=#6
     possessed_class=CommonTextBlock
     tracks [
      {
       type=MovieSceneFloatTrack
       display_name=Render Opacity
       sections [{type=MovieSceneFloatSection, channels [None {type=MovieSceneScriptingFloatChannel, const [0,1]}]}]
      }
      {
       type=MovieScene2DTransformTrack
       sections [{
         type=MovieScene2DTransformSection
         channels [
          Translation.X {type=MovieSceneScriptingFloatChannel, const [128,1]}
          Translation.Y {type=MovieSceneScriptingFloatChannel, const [0,1]}
          Angle {type=MovieSceneScriptingFloatChannel}
          Scale.X {type=MovieSceneScriptingFloatChannel}
          Scale.Y {type=MovieSceneScriptingFloatChannel}
          Shear.X {type=MovieSceneScriptingFloatChannel}
          Shear.Y {type=MovieSceneScriptingFloatChannel}
         ]
        }]
      }
     ]
    }
    SeasonPassBadgeIcon {
     id=#7
     possessed_class=SeasonPassBadgeIcon_C
     tracks [
      {
       type=MovieSceneFloatTrack
       display_name=Render Opacity
       sections [{type=MovieSceneFloatSection, channels [None {type=MovieSceneScriptingFloatChannel, const [0,1]}]}]
      }
      {
       type=MovieScene2DTransformTrack
       sections [{
         type=MovieScene2DTransformSection
         channels [
          Translation.X {type=MovieSceneScriptingFloatChannel, const [-32,1]}
          Translation.Y {type=MovieSceneScriptingFloatChannel, const [0,1]}
          Angle {type=MovieSceneScriptingFloatChannel}
          Scale.X {type=MovieSceneScriptingFloatChannel}
          Scale.Y {type=MovieSceneScriptingFloatChannel}
          Shear.X {type=MovieSceneScriptingFloatChannel}
          Shear.Y {type=MovieSceneScriptingFloatChannel}
         ]
        }]
      }
     ]
    }
   ]}
 }
 FadeAnim {
  binding_count=4
  length_seconds=0.5
  object=&self:FadeAnim
  sequence {bindings [
    Text_Body {
     id=#2
     possessed_class=CommonTextBlock
     tracks [
      {
       type=MovieScene2DTransformTrack
       sections [{
         type=MovieScene2DTransformSection
         channels [
          Translation.X {type=MovieSceneScriptingFloatChannel, keys [[7,0],[10,-128]]}
          Translation.Y {type=MovieSceneScriptingFloatChannel, const [0,2]}
          Angle {type=MovieSceneScriptingFloatChannel}
          Scale.X {type=MovieSceneScriptingFloatChannel}
          Scale.Y {type=MovieSceneScriptingFloatChannel}
          Shear.X {type=MovieSceneScriptingFloatChannel}
          Shear.Y {type=MovieSceneScriptingFloatChannel}
         ]
        }]
      }
      {
       type=MovieSceneFloatTrack
       display_name=Render Opacity
       sections [{type=MovieSceneFloatSection, channels [None {type=MovieSceneScriptingFloatChannel, keys [[7,1],[10,0]]}]}]
      }
     ]
    }
    Text_SubHeader {
     id=#3
     possessed_class=CommonTextBlock
     tracks [
      {
       type=MovieScene2DTransformTrack
       sections [{
         type=MovieScene2DTransformSection
         channels [
          Translation.X {type=MovieSceneScriptingFloatChannel, keys [[6,0],[9,-128]]}
          Translation.Y {type=MovieSceneScriptingFloatChannel, const [0,2]}
          Angle {type=MovieSceneScriptingFloatChannel}
          Scale.X {type=MovieSceneScriptingFloatChannel}
          Scale.Y {type=MovieSceneScriptingFloatChannel}
          Shear.X {type=MovieSceneScriptingFloatChannel}
          Shear.Y {type=MovieSceneScriptingFloatChannel}
         ]
        }]
      }
      {
       type=MovieSceneFloatTrack
       display_name=Render Opacity
       sections [{type=MovieSceneFloatSection, channels [None {type=MovieSceneScriptingFloatChannel, keys [[6,1],[9,0]]}]}]
      }
     ]
    }
    Text_Header {
     id=#8
     possessed_class=CommonTextBlock
     tracks [
      {
       type=MovieSceneFloatTrack
       display_name=Render Opacity
       sections [{type=MovieSceneFloatSection, channels [None {type=MovieSceneScriptingFloatChannel, keys [[3,1],[6,0]]}]}]
      }
      {
       type=MovieScene2DTransformTrack
       sections [{
         type=MovieScene2DTransformSection
         channels [
          Translation.X {type=MovieSceneScriptingFloatChannel, keys [[3,0],[6,-128]]}
          Translation.Y {type=MovieSceneScriptingFloatChannel, const [0,2]}
          Angle {type=MovieSceneScriptingFloatChannel}
          Scale.X {type=MovieSceneScriptingFloatChannel}
          Scale.Y {type=MovieSceneScriptingFloatChannel}
          Shear.X {type=MovieSceneScriptingFloatChannel}
          Shear.Y {type=MovieSceneScriptingFloatChannel}
         ]
        }]
      }
     ]
    }
    SeasonPassBadgeIcon {
     id=#9
     possessed_class=SeasonPassBadgeIcon_C
     tracks [
      {
       type=MovieSceneFloatTrack
       display_name=Render Opacity
       sections [{type=MovieSceneFloatSection, channels [None {type=MovieSceneScriptingFloatChannel, keys [[0,1],[3,0]]}]}]
      }
      {
       type=MovieScene2DTransformTrack
       sections [{
         type=MovieScene2DTransformSection
         channels [
          Translation.X {type=MovieSceneScriptingFloatChannel, keys [[0,0],[3,-64]]}
          Translation.Y {type=MovieSceneScriptingFloatChannel, const [0,2]}
          Angle {type=MovieSceneScriptingFloatChannel}
          Scale.X {type=MovieSceneScriptingFloatChannel}
          Scale.Y {type=MovieSceneScriptingFloatChannel}
          Shear.X {type=MovieSceneScriptingFloatChannel}
          Shear.Y {type=MovieSceneScriptingFloatChannel}
         ]
        }]
      }
     ]
    }
   ]}
 }
]
graphs [
 EventGraph {nodes [
   n1 = Set Custom Background From Season Pass!() -> (OutputDelegate, then, Pass:Season Pass VM Object Reference) {is=title K2Node_CustomEvent}
   > n2 = Is Valid(InputObject=@n1.Pass)
   Set Custom Background(execute=@n2.Is Not Valid, Custom Background:Soft Object Reference) -> (Output_Get:Soft Object Reference)
   n3 = Break Season Pass Theme Data!(SeasonPassThemeData:"Season Pass Theme Data Structure (by ref)"=GetThemeData!(self:Season Pass VM Object Reference=@n1.Pass) -> (ReturnValue:"Season Pass Theme Data Structure (by ref)")) -> (
    TitleColor:Linear Color Structure
    TitleMaterial:Soft Object Reference
    TitleOutline:Font Outline Settings Structure
    SeasonColor:Linear Color Structure
    ProgressionColor:Linear Color Structure
    OverlayColor:Linear Color Structure
    VaultBackgroundMaterial:Soft Object Reference
    GiftboxVaultBackgroundMaterial:Soft Object Reference
    UIBackgroundMaterial:Soft Object Reference
    UpsellBackgroundMaterial:Soft Object Reference
    CrewCategoryBackgroundMaterial:Soft Object Reference
    CustomLogo:Soft Object Reference
    CustomSeasonTitle:Text
    CustomSeasonTitleColor:Linear Color Structure
    Upsell2BackgrounMaterial:Soft Object Reference
    ImageReferenceTable:Fort Image Reference Table Structure
    ColorReferenceTable:Fort Color Reference Table Structure
    TextReferenceTable:Fort Text Reference Table Structure
   ) {is=title K2Node_BreakStruct}
   Set Custom Background(
    execute=@n2.Is Valid
    Custom Background=Select(
     Option 0:Soft Object Reference=@n3.VaultBackgroundMaterial
     Option 1:Soft Object Reference=@n3.GiftboxVaultBackgroundMaterial
     Index=IsValidSoftObjectReference(SoftObjectReference=@n3.GiftboxVaultBackgroundMaterial)
    ) -> (ReturnValue:Soft Object Reference)
   )
   Event PreConstruct()
   > Set Header Color(Header Color:Slate Color Structure=Get ColorAndOpacity!(self:Text Object Reference=Get Text_Header() -> (Text_Header:Common Text Object Reference)) -> (ColorAndOpacity:Slate Color Structure)) -> (Output_Get:Slate Color Structure)
   > Set SubHeader Color(SubHeader Color:Slate Color Structure=Get ColorAndOpacity(self=Get Text_SubHeader() -> (Text_SubHeader:Common Text Object Reference))) -> (Output_Get:Slate Color Structure)
   > Set Body Color(Body Color:Slate Color Structure=Get ColorAndOpacity(self=Get Text_Body() -> (Text_Body:Common Text Object Reference))) -> (Output_Get:Slate Color Structure)
  ]}
 Get Color Override {nodes [
   n1 = Get Color Override!() -> (then, Color:Slate Color Structure, Color Override:Slate Color Structure, Use Color Override:bool=~) {is=title K2Node_FunctionEntry}
   > Return Node!(
    execute
    NewParam1:Slate Color Structure=Select(Option 0:Slate Color Structure=@n1, Option 1:Slate Color Structure=@n1.Color Override, Index=@n1.Use Color Override) -> (ReturnValue:Slate Color Structure)
   )
  ]}
 Get Header Text {nodes [
   n1 = Get Header Text!() -> (then, Gift Box Text:Text, Item Text:Text) {is=title K2Node_FunctionEntry}
   > Return Node!(execute, Result:Text=SelectText(A=@n1.Item Text, B=@n1, bPickA=TextIsEmpty(InText=@n1)))
  ]}
 Get Season Pass {nodes [
   n1 = Get Season Pass!() -> (then, Gift Box:Gift Box VM Object Reference, Season Pass System:Season Pass System VM Object Reference) {is=title K2Node_FunctionEntry}
   > n2 = Branch(Condition=AND Boolean(A=IsValid(Object=@n1), B=IsValid(Object=@n1.Season Pass System)))
   n3 = Get GiftBoxItem!(self:Gift Box VM Object Reference=Get Gift Box() -> (Gift Box:Gift Box VM Object Reference)) -> (GiftBoxItem:Fort Gift Box Item Object Reference)
   > Return Node(Result:Season Pass VM Object Reference=FindSeasonPass!(
     self:Season Pass System VM Object Reference=Get Season Pass System() -> (Season Pass System:Season Pass System VM Object Reference)
     TemplateId:String=TryGetParamString!(self:Fort Gift Box Item Object Reference=@n3, ParamName:String=seasonTemplateId) -> (OutValue:String, ReturnValue:bool)
    ) -> (ReturnValue:Season Pass VM Object Reference))
   Return Node(execute=@n2.else)
  ]}
 GetCustomBackground {nodes [
   GetCustomBackground() {is=title K2Node_FunctionEntry}
   > Return Node!(execute, Background:Soft Object Reference=Get Custom Background() -> (Custom Background:Soft Object Reference))
  ]}
 HandleGiftBoxSet {nodes [
   n1 = HandleGiftBoxSet!() -> (then, GiftBox:Gift Box VM Object Reference) {is=title K2Node_FunctionEntry}
   > n2 = Get!(execute:Exec, self:Gift Box VM Object Reference=@n1) -> (then:Exec, else:Exec, GiftBoxItem:Fort Gift Box Item Object Reference)
   n3 = TryGetParamBool!(self:Fort Gift Box Item Object Reference=@n2, ParamName:String=autoSpend) -> (OutValue:bool, ReturnValue:bool)
   > n4 = Branch(Condition=@n3)
   > SetVisibility(
    self=Get AutoClaimStatus() -> (AutoClaimStatus:WBP Gift Box Auto Claimed Status Object Reference)
    InVisibility=Select(Option 0:ESlateVisibility Enum=Collapsed, Option 1:ESlateVisibility Enum=HitTestInvisible, Index=@n3.ReturnValue) -> (ReturnValue:ESlateVisibility Enum)
   )
   SetVisibility(execute=[@n5.then, @n4.else], self=Get AutoClaimStatus(), InVisibility=Collapsed)
   n5 = Set Custom Background From Season Pass!(execute=@n2.else, self, Pass:Season Pass VM Object Reference) -> (then) {is=title K2Node_CallFunction}
  ]}
 HandleInitialState {nodes [
   HandleInitialState() {is=title K2Node_FunctionEntry}
   > PlayAnimationTimeRange!(
    execute
    self:User Widget Object Reference
    InAnimation:Widget Animation Object Reference=Get InitialState() -> (InitialState:Widget Animation Object Reference)
    StartAtTime:float=0
    EndAtTime:float=0
    NumLoopsToPlay:int=1
    PlayMode:EUMGSequencePlayMode Enum
    PlaybackSpeed:float=1
    bRestoreState:bool=false
   ) -> (then, ReturnValue:Widget Animation Handle Structure)
  ]}
 HandleIntro {nodes [
   HandleIntro() {is=title K2Node_FunctionEntry}
   > StopAnimation(InAnimation=Get FadeAnim() -> (FadeAnim:Widget Animation Object Reference))
   > PlayAnimationForward(InAnimation=Get IntroAnim() -> (IntroAnim:Widget Animation Object Reference), PlaybackSpeed=1.5)
  ]}
 HandleOutro {nodes [
   HandleOutro() {is=title K2Node_FunctionEntry}
   > PlayAnimationForward(InAnimation=Get FadeAnim(), PlaybackSpeed=1.5)
  ]}
 Set Body Text {nodes [
   n1 = Set Body Text!() -> (then, Text:Text) {is=title K2Node_FunctionEntry}
   n2 = Get Text_Body()
   > "SetText (Text)"(self=@n2, InText=@n1)
   > SetVisibility(self=@n2, InVisibility=Select(Option 0=SelfHitTestInvisible, Option 1=Collapsed, Index=TextIsEmpty(InText=@n1)))
  ]}
 Set Season Pass {nodes [
   Set Season Pass!() -> (then, Pass:Season Pass VM Object Reference) {is=title K2Node_FunctionEntry}
   > Is Valid(InputObject=Get GiftBoxVM() -> (GiftBoxVM:Gift Box VM Object Reference))
   > n1 = Is Valid(InputObject=Get Pass() -> (Pass:Season Pass VM Object Reference))
   n2 = Get SeasonPassBadgeIcon() -> (SeasonPassBadgeIcon:Season Pass Badge Icon Object Reference)
   n3 = Get GiftBoxVM()
   SetLargeIcon!(
    execute=@n1.Is Not Valid
    self:Season Pass Badge Icon Object Reference=@n2
    Icon:Soft Object Reference=Get HeaderImage!(self:Gift Box VM Object Reference=@n3) -> (HeaderImage:Soft Object Reference)
   ) -> (then)
   > SetDoesOwnPass!(
    execute
    self:Season Pass Badge Icon Object Reference=@n2
    DoesOwnPass:bool=TryGetParamBool(self=Get GiftBoxItem(self=@n3), ParamName=seasonOwned)
   ) -> (then)
   > Set Custom Background From Season Pass()
   n4 = Get Pass()
   SetSeasonPassVM!(execute=@n1.Is Valid, self:Season Pass Badge Icon Object Reference=@n2, Viewmodel:Season Pass VM Object Reference=@n4) -> (then)
   > Set Custom Background From Season Pass(Pass=@n4)
  ]}
 Set Sub Header Text {nodes [
   n1 = Set Sub Header Text!() -> (then, Text:Text) {is=title K2Node_FunctionEntry}
   n2 = Get Text_SubHeader()
   > "SetText (Text)"(self=@n2, InText=@n1)
   > SetVisibility(self=@n2, InVisibility=Select(Option 0=SelfHitTestInvisible, Option 1=Collapsed, Index=TextIsEmpty(InText=@n1)))
  ]}
]
variables [
 Badge Color {type=SlateColor, default {colorUseRule=UseColor_Specified, specifiedColor=<rgba 1 0 1 1>}}
 Body Color {type=SlateColor, default {colorUseRule=UseColor_Specified, specifiedColor=<rgba 1 0 1 1>}}
 Custom Background {type=Object, default=null}
 Header Color {type=SlateColor, default {colorUseRule=UseColor_Specified, specifiedColor=<rgba 1 0 1 1>}}
 SubHeader Color {type=SlateColor, default {colorUseRule=UseColor_Specified, specifiedColor=<rgba 1 0 1 1>}}
]
widgets {
 info {inheritedWidgetCount=0, parentClass=&CommonUserWidget, rootWidgetClass=&HorizontalBox, widgetCount=7}
 widgets [
  HB_Content {widgetClassPath=&HorizontalBox, widget=HB_Content}
  SeasonPassBadgeIcon {
   widgetClassPath=/SeasonPass/Widgets/SeasonPassBadgeIcon
   bIsVariable=true
   parent=HB_Content
   widget=SeasonPassBadgeIcon
   properties {size=<128 128>}
   slot {class=HorizontalBoxSlot, properties {horizontalAlignment=HAlign_Left, verticalAlignment=VAlign_Top, padding=<ltrb 0 0 16 0>}}
  }
  VerticalBox_1 {
   widgetClassPath=&VerticalBox
   parent=HB_Content
   widget=VerticalBox_1
   slot {class=HorizontalBoxSlot, properties {verticalAlignment=VAlign_Center, size {sizeRule=Fill, value=1}}}
  }
  Text_Header {
   widgetClassPath=&CommonTextBlock
   bIsVariable=true
   parent=VerticalBox_1
   widget=Text_Header
   properties {
    autoWrapText=true
    lineHeightPercentage=0.85
    text="{HEADER STRING}"
    textTransformPolicy=ToUpper
    font {
     bForceMonospaced=false
     bMaterialIsStencil=false
     fontMaterial=null
     fontObject=&HeadingNow
     letterSpacing=0
     monospacedWidth=1
     size=32.25
     skewAmount=0
     typefaceFontName=76 Bold Italic
     outlineSettings {
      bApplyOutlineToDropShadows=false
      bMiteredCorners=false
      bSeparateFillAlpha=false
      outlineMaterial=null
      outlineSize=0
      outlineColor=<rgba 0 0 0 1>
     }
    }
   }
   slot {class=VerticalBoxSlot, properties {verticalAlignment=VAlign_Top}}
  }
  Text_SubHeader {
   widgetClassPath=&CommonTextBlock
   bIsVariable=true
   parent=VerticalBox_1
   widget=Text_SubHeader
   properties {
    autoWrapText=true
    text="[ph]SUBTITLE LOREM IPSUM DOLOR"
    visibility=Collapsed
    font {
     bForceMonospaced=false
     bMaterialIsStencil=false
     fontMaterial=null
     fontObject=null
     letterSpacing=0
     monospacedWidth=1
     size=24
     skewAmount=0
     typefaceFontName=None
     outlineSettings {
      bApplyOutlineToDropShadows=false
      bMiteredCorners=false
      bSeparateFillAlpha=false
      outlineMaterial=null
      outlineSize=0
      outlineColor=<rgba 0 0 0 1>
     }
    }
   }
   slot {class=VerticalBoxSlot, properties {verticalAlignment=VAlign_Top, padding=<ltrb 0 8 0 8>}}
  }
  Text_Body {
   widgetClassPath=&CommonTextBlock
   bIsVariable=true
   parent=VerticalBox_1
   widget=Text_Body
   properties {
    autoWrapText=true
    flowDirectionPreference=LeftToRight
    lineHeightPercentage=1.29
    text="{AssetBodyString} Lorem IpsumDolor Amit Lorem  Ipsum Dolor Amit Lorem Ipsum Dolor Amit"
    colorAndOpacity {colorUseRule=UseColor_Specified, specifiedColor=<rgba 0.5776 0.5776 0.5776 1>}
    font {
     bForceMonospaced=false
     bMaterialIsStencil=false
     fontMaterial=null
     fontObject=&HeadingNow
     letterSpacing=0
     monospacedWidth=1
     size=18
     skewAmount=0
     typefaceFontName=74 Regular
     outlineSettings {
      bApplyOutlineToDropShadows=false
      bMiteredCorners=false
      bSeparateFillAlpha=false
      outlineMaterial=null
      outlineSize=0
      outlineColor=<rgba 0 0 0 1>
     }
    }
    shapedTextOptions {bOverride_TextFlowDirection=true, bOverride_TextShapingMethod=false, textFlowDirection=LeftToRight, textShapingMethod=Auto}
   }
   slot {class=VerticalBoxSlot, properties {verticalAlignment=VAlign_Top, padding=<ltrb 0 8 0 0>}}
  }
  AutoClaimStatus {
   widgetClassPath=/GiftBox/UI/WBP_GiftBox_AutoClaimedStatus
   bIsVariable=true
   parent=VerticalBox_1
   widget=AutoClaimStatus
   slot {class=VerticalBoxSlot, properties {horizontalAlignment=HAlign_Left, padding=<ltrb 0 16 0 0>}}
  }
 ]
}
