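# Translucent UI material for the animated LEGO Pass purchase transition, wiping between the default and alternate pass icons along an angled, bendable edge.
# ShineSwipe drives that wipe mask, switching silver-blue to gold palettes and positioning a glowing divider beam.
# Panned noise dots distort icon UVs, while soft and electric noise, warped by icon brightness, layer through color dodge.
# A separate sheen and optional flat tint finish the color; opacity combines icon alpha and beam.
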Material /SeasonPass/Giftbox/M_UI_LegoPassPurchase_Animated_Transition
graph {
 properties {
  bUsedWithStaticMesh=false
  blendMode=BLEND_Translucent
  materialDomain=MD_UI
  pixelDepthOffsetMode=PDOM_Legacy
  usedShadingModels=Default Lit
 }
 nodes [
  n1 = Add(
   A=Rotator(Coordinate=TextureCoordinate(), Time=ScalarParameter() {defaultValue=0.125, group=Animate, parameterName=SheenAngle}) {speed=6.2832}
   B=ScalarParameter() {defaultValue=-0.5, group=Animate, parameterName=SheenSwipe}
  )
  n2 = ComponentMask(None=Sine(None=Multiply(A=@n1) {constB=0.5})) {g=true}
  n3 = Saturate(None=Power(Base=Saturate(None=@n2)) {constExponent=6})
  n4 = VectorParameter() {group=Color, parameterName=SIlverColor_2, defaultValue=<rgba 0 0.0833 1 1>}
  n5 = VectorParameter() {group=Color, parameterName=SilverColor_1, defaultValue=<rgba 0.2 0.8 1 1>}
  n6 = TextureCoordinate()
  n7 = Constant() {r=0.666}
  n8 = Constant()
  n9 = Constant() {r=0.075}
  n10 = ScalarParameter() {group=Animate, parameterName=ShineBend}
  n11 = AppendVector(
   A=Constant()
   B=Multiply(
    A=MaterialFunctionCall(Radius=@n7) {materialFunction=&RadialGradientExponential}.RadialGradientExponential
    B=LinearInterpolate(A=@n8, B=@n9, Alpha=@n10)
   )
  )
  n12 = Add(
   A=Rotator(Coordinate=Add(A=@n6, B=@n11), Time=ScalarParameter() {defaultValue=0.125, group=Animate, parameterName=ShineAngle}) {speed=6.2832}
   B=ScalarParameter() {group=Animate, parameterName=ShineSwipe}
  )
  n13 = ComponentMask(None=Sine(None=Multiply(A=@n12) {constB=0.5})) {g=true}
  n14 = LinearInterpolate(A=Constant() {r=-12}, B=Constant() {r=1}, Alpha=@n13)
  n15 = ScalarParameter() {group=Treatment, parameterName=BeamThiccness}
  n16 = Multiply(
   A=Saturate(None=LinearInterpolate(A=@n14, B=@n13, Alpha=@n15))
   B=ScalarParameter() {defaultValue=0.25, group=Treatment, parameterName=BeamGlow}
  )
  n17 = VectorParameter() {group=Color, parameterName=GoldColor_2, defaultValue=<rgba 1 0.1 0 1>}
  n18 = VectorParameter() {group=Color, parameterName=GoldColor_1, defaultValue=<rgba 1 0.9333 0.2 1>}
  n19 = Add(A=TextureCoordinate(), B=@n11) {constB=0}
  n20 = ScalarParameter() {defaultValue=0.125, group=Animate, parameterName=ShineAngle}
  n21 = Add(A=Add(A=Rotator(Coordinate=@n19, Time=@n20) {speed=6.2832}) {constB=-0.5}, B=ScalarParameter() {group=Animate, parameterName=ShineSwipe})
  n22 = Saturate(None=Multiply(A=ComponentMask(None=@n21) {g=true}) {constB=64})
  n23 = TextureCoordinate()
  n24 = AppendVector(
   A=ScalarParameter() {defaultValue=-0.25, group=Noise, parameterName=DotsGoldSmall_USpeed}
   B=ScalarParameter() {defaultValue=-0.5, group=Noise, parameterName=DotsGoldSmall_VSpeed}
  )
  n25 = TextureCoordinate() {uTiling=2, vTiling=2}
  n26 = ScalarParameter() {defaultValue=0.75, group=Noise, parameterName=DotsNoiseScale}
  n27 = Multiply(A=@n25, B=@n26)
  n28 = AppendVector(
   A=ScalarParameter() {defaultValue=-0.5, group=Noise, parameterName=DotsGoldLarge_USpeed}
   B=ScalarParameter() {defaultValue=-1, group=Noise, parameterName=DotsGoldLarge_VSpeed}
  )
  n29 = DotProduct(
   A=TextureSample(UVs=Panner(Coordinate=@n25, Speed=@n24) {speedX=-0.075, speedY=0.125}) {samplerType=SAMPLERTYPE_Grayscale, texture=&LowResBlurredNoise}.R
   B=TextureSample(UVs=Panner(Coordinate=@n27, Speed=@n28) {speedX=-0.075, speedY=0.125}) {samplerType=SAMPLERTYPE_Grayscale, texture=&LowResBlurredNoise}.G
  )
  n30 = Add(A=Multiply(A=ScalarParameter() {parameterName=ShineSwipe}) {constB=-1}) {constB=0.5}
  n31 = ScalarParameter() {defaultValue=1, group=Treatment, parameterName=DotsIntensity}
  n32 = Power(Base=MaterialFunctionCall(CenterPosition=AppendVector(A=@n30, B=@n30), Radius=Constant() {r=0.5}, Density=Multiply(B=@n31) {constA=2.333}) {materialFunction=&RadialGradientExponential}.RadialGradientExponential)
  n33 = AppendVector(
   A=ScalarParameter() {defaultValue=0.25, group=Noise, parameterName=DotsSilverSmall_USpeed}
   B=ScalarParameter() {defaultValue=0.5, group=Noise, parameterName=DotsSilverSmall_VSpeed}
  )
  n34 = Multiply(A=@n25, B=@n26)
  n35 = AppendVector(
   A=ScalarParameter() {defaultValue=0.5, group=Noise, parameterName=DotsSilverLarge_USpeed}
   B=ScalarParameter() {defaultValue=1, group=Noise, parameterName=DotsSilverLarge_VSpeed}
  )
  n36 = DotProduct(
   A=TextureSample(UVs=Panner(Coordinate=@n25, Speed=@n33) {speedX=-0.075, speedY=0.125}) {samplerType=SAMPLERTYPE_Grayscale, texture=&LowResBlurredNoise}.R
   B=TextureSample(UVs=Panner(Coordinate=@n34, Speed=@n35) {speedX=-0.075, speedY=0.125}) {samplerType=SAMPLERTYPE_Grayscale, texture=&LowResBlurredNoise}.G
  )
  n37 = LinearInterpolate(A=DotProduct(A=Multiply(A=@n36) {constB=4}, B=@n32), B=DotProduct(A=Multiply(A=@n29) {constB=4}, B=@n32), Alpha=@n22)
  n38 = Multiply(
   A=Saturate(None=@n37)
   B=LinearInterpolate(A=Constant() {r=0.005}, B=Constant() {r=0.1}, Alpha=ScalarParameter() {group=Treatment, parameterName=WaveDistortAmount})
  ) {constB=0.0125}
  n39 = TextureSample(UVs=Add(A=@n23, B=@n38)) {texture=/Game/UI/Foundation/Textures/Icons/Items/Passes/T_UI_Nexus_LegoPass_Default_L}
  n40 = TextureSample(UVs=Add(A=@n23, B=@n38)) {texture=/Game/UI/Foundation/Textures/Icons/Items/Passes/T_UI_Nexus_LegoPass_L}
  n41 = LinearInterpolate(A=@n39.RGB, B=@n40.RGB, Alpha=@n22)
  n42 = LinearInterpolate(
   A=@n41
   B=LinearInterpolate(A=LinearInterpolate(A=@n4.RGB, B=@n5.RGB, Alpha=@n16), B=LinearInterpolate(A=@n17.RGB, B=@n18.RGB, Alpha=@n16), Alpha=@n22)
   Alpha=@n16
  )
  n43 = Add(A=Multiply(A=ScalarParameter() {group=Animate, parameterName=ShineSwipe}) {constB=-1}) {constB=0.5}
  n44 = Multiply(A=TextureCoordinate(), B=ScalarParameter() {defaultValue=0.5, group=Noise, parameterName=SoftNoise_Tiling})
  n45 = LinearInterpolate(A=@n39.A, B=@n40.A, Alpha=@n22)
  n46 = DotProduct(A=@n41, B=@n45)
  n47 = ScalarParameter() {defaultValue=3, group=Treatment, parameterName=NoiseDistortion_Contrast}
  n48 = Multiply(A=Saturate(None=Power(Base=@n46, Exp=@n47)), B=ScalarParameter() {defaultValue=0.15, group=Treatment, parameterName=NoiseDistortion_Amount})
  n49 = ScalarParameter() {defaultValue=2.333, group=Noise, parameterName=NoiseRounded-ness}
  n50 = Multiply(A=OneMinus(None=MaterialFunctionCall(Density=@n49) {materialFunction=&RadialGradientExponential}.RadialGradientExponential)) {constB=0.333}
  n51 = Panner(
   Coordinate=Add(A=Add(A=@n44, B=@n48), B=@n50)
   Speed=AppendVector(
    A=ScalarParameter() {defaultValue=-0.075, group=Noise, parameterName=SoftNoise_USpeed}
    B=ScalarParameter() {defaultValue=0.125, group=Noise, parameterName=SoftNoise_VSpeed}
   )
  ) {speedX=-0.075, speedY=0.125}
  n52 = Multiply(A=TextureCoordinate(), B=ScalarParameter() {defaultValue=1.5, group=Noise, parameterName=ElectricNoise_Tiling})
  n53 = Panner(
   Coordinate=Add(A=Add(A=@n52, B=@n48), B=@n50)
   Speed=AppendVector(
    A=ScalarParameter() {defaultValue=0.05, group=Noise, parameterName=ElectricNoise_USpeed}
    B=ScalarParameter() {defaultValue=0.025, group=Noise, parameterName=ElectricNoise_VSpeed}
   )
  ) {speedX=0.05, speedY=0.025}
  n54 = Multiply(
   A=DotProduct(
    A=TextureSample(UVs=@n51) {samplerType=SAMPLERTYPE_Grayscale, texture=&LowResBlurredNoise}.R
    B=TextureSample(UVs=@n53) {samplerType=SAMPLERTYPE_Grayscale, texture=&LowResBlurredNoise}.R
   )
   B=ScalarParameter() {defaultValue=0.6667, group=Treatment, parameterName=SoftNoise_Intensity}
  ) {constB=0.6666}
  n55 = ScalarParameter() {defaultValue=3, group=Treatment, parameterName=NoiseIntensity_Edges}
  n56 = DotProduct(
   A=TextureSample(UVs=@n51) {samplerType=SAMPLERTYPE_LinearColor, texture=/Game/Characters/Player/Male/Medium/Bodies/M_MED_SRDrift/Textures/T_ElectricNoise}.R
   B=TextureSample(UVs=@n53) {samplerType=SAMPLERTYPE_LinearColor, texture=&T_ElectricNoise}.R
  )
  n57 = Multiply(A=@n56, B=@n55)
  n58 = ScalarParameter() {defaultValue=1, group=Treatment, parameterName=NoiseContrast_Edges}
  n59 = ScalarParameter() {defaultValue=6, group=Treatment, parameterName=NoiseIntensity_Center}
  n60 = Multiply(A=@n56, B=@n59) {constB=4}
  n61 = LinearInterpolate(
   A=Add(A=Multiply(A=@n54, B=@n55), B=Power(Base=@n57, Exp=@n58))
   B=Add(A=Multiply(A=@n54, B=@n59) {constB=4}, B=Power(Base=@n60, Exp=@n58))
   Alpha=MaterialFunctionCall(CenterPosition=AppendVector(A=@n43, B=@n43), Radius=Constant() {r=0.4}, Density=Constant() {r=2}) {materialFunction=&RadialGradientExponential}.RadialGradientExponential
  )
  n62 = LinearInterpolate(
   A=VectorParameter() {group=Color, parameterName=SIlverColor_2, defaultValue=<rgba 0 0.0833 1 1>}.RGB
   B=VectorParameter() {group=Color, parameterName=SilverColor_1, defaultValue=<rgba 0.2 0.8 1 1>}.RGB
   Alpha=@n61
  )
  n63 = LinearInterpolate(
   A=VectorParameter() {group=Color, parameterName=GoldColor_2, defaultValue=<rgba 1 0.1 0 1>}.RGB
   B=VectorParameter() {group=Color, parameterName=GoldColor_1, defaultValue=<rgba 1 0.9333 0.2 1>}.RGB
   Alpha=@n61
  )
  n64 = Saturate(None=LinearInterpolate(B=LinearInterpolate(A=@n62, B=@n63, Alpha=@n22), Alpha=@n61))
  n65 = Constant() {r=-36}
  n66 = Constant() {r=1}
  n67 = Saturate(None=LinearInterpolate(A=LinearInterpolate(A=@n65, B=@n66, Alpha=@n13), B=@n13, Alpha=ScalarParameter() {group=Treatment, parameterName=DividerLine_Thiccness}))
  n68 = DotProduct(A=Saturate(None=Add(A=@n39.A, B=@n40.A)), B=Saturate(None=Power(Base=@n67) {constExponent=6}))
  n69 = Saturate(None=Add(A=@n37, B=@n68))
  n70 = LinearInterpolate(
   A=LinearInterpolate(A=VectorParameter() {group=Color, parameterName=SilverBeam_Color, defaultValue=<rgba 0.25 0.8125 1 1>}.RGB, B=Constant() {r=1}, Alpha=@n69)
   B=LinearInterpolate(A=VectorParameter() {group=Color, parameterName=GoldBeam_Color, defaultValue=<rgba 1 0.9375 0.25 1>}.RGB, B=Constant() {r=1}, Alpha=@n69)
   Alpha=@n22
  )
  n71 = Add(
   A=MaterialFunctionCall!(
    Blend=LinearInterpolate(B=@n70, Alpha=@n69)
    Base=MaterialFunctionCall(Blend=@n64, Base=@n42) {materialFunction=/Engine/Functions/Engine_MaterialFunctions03/Blends/Blend_ColorDodge}.Result
   ) {materialFunction=&Blend_ColorDodge}.Result
   B=Multiply(A=@n64, B=ScalarParameter() {defaultValue=0.75, group=Treatment, parameterName=PureNoiseTint_Amount})
  )
  LinearInterpolate(
   A=Saturate(None=MaterialFunctionCall(Blend=@n3, Base=@n71) {materialFunction=&Blend_ColorDodge}.Result)
   B=VectorParameter() {group=Color, parameterName=PureTintColor, defaultValue=<rgba 1 0.8417 0.05 1>}.RGB
   Alpha=ScalarParameter() {group=Animate, parameterName=UsePureTint}
  )
  LinearInterpolate(
   A=LinearInterpolate(A=Constant() {r=-12}, B=Constant() {r=1}, Alpha=@n2)
   B=@n2
   Alpha=ScalarParameter() {defaultValue=1, group=Treatment, parameterName=SheenLine_Thiccness}
  )
  n72 = Power(Base=@n68) {constExponent=3}
  Saturate(None=Add(A=@n45, B=Multiply(A=@n72) {constB=4}))
 ]
}
